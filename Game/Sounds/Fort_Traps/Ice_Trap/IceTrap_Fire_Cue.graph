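# Sound cue for the ice trap's firing sound, using the Fort_Traps sound class and attenuation.
# A quality-level root picks one of four branches: the first mixes a random pick of three activate waves with a volume-modulated layer, and the second mixes one fixed wave with that layer.
# The last two branches drop the layer and play that single wave with pitch and volume variation.
# Concurrency override allows one instance and blocks new ones.
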SoundCue /Game/Sounds/Fort_Traps/Ice_Trap/IceTrap_Fire_Cue
root=QualityLevel_1
nodes [
 {id=Mixer_2, children [Random_0,Modulator_4], properties {inputVolume [1,0.9]}}
 {id=Mixer_3, children [WavePlayer_13, Modulator_7], properties {inputVolume [1,0.7]}}
 {id=Modulator_4, children [WavePlayer_10], properties {volumeMax=1.2, volumeMin=1}}
 {id=Modulator_5, children [WavePlayer_15], properties {pitchMax=1.1, pitchMin=0.9, volumeMax=1.2, volumeMin=1}}
 {id=Modulator_6, children [WavePlayer_16], properties {pitchMax=1.1, pitchMin=0.9, volumeMax=1.2, volumeMin=1}}
 {id=Modulator_7, children [WavePlayer_14], properties {volumeMax=1.2, volumeMin=1}}
 {id=QualityLevel_1, children [Mixer_2,Mixer_3,Modulator_5,Modulator_6]}
 {id=Random_0, children [WavePlayer_2,WavePlayer_5,WavePlayer_8], properties {weights [1,1,1]}}
 {id=WavePlayer_10, properties {soundWaveAssetPtr=&/IceTrap_Activate_Layer_01}}
 {id=WavePlayer_13, properties {soundWaveAssetPtr=&/IceTrap_Activate_03}}
 {id=WavePlayer_14, properties {soundWaveAssetPtr=&IceTrap_Activate_Layer_01}}
 {id=WavePlayer_15, properties {soundWaveAssetPtr=&IceTrap_Activate_03}}
 {id=WavePlayer_16, properties {soundWaveAssetPtr=&IceTrap_Activate_03}}
 {id=WavePlayer_2, properties {soundWaveAssetPtr=&/IceTrap_Activate_01}}
 {id=WavePlayer_5, properties {soundWaveAssetPtr=&/IceTrap_Activate_02}}
 {id=WavePlayer_8, properties {soundWaveAssetPtr=&IceTrap_Activate_03}}
]
properties {
 attenuationSettings=&Fort_Traps_Att
 bOverrideConcurrency=true
 maxDistance=6000
 soundClassObject=&Fort_Traps
 volumeMultiplier=0.7
 concurrencyOverrides {maxCount=1, resolutionRule=PreventNew}
}
