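# Sound cue for the lock-on sound of the Ice Trap in the Fort traps set.
# A single wave player feeds a modulator that randomizes pitch and volume, with cue-level multipliers on top.
# It uses the shared traps attenuation and sound class.
# A concurrency override allows only one instance at a time and prevents new ones from starting while it plays.
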SoundCue /Game/Sounds/Fort_Traps/Ice_Trap/IceTrap_Lockon_Cue
root=Modulator_0
nodes [
 {id=Modulator_0, children [WavePlayer_0], properties {pitchMax=1.2, volumeMax=1.5}}
 {id=WavePlayer_0, properties {soundWaveAssetPtr=&/IceTrap_Lockon_01}}
]
properties {
 attenuationSettings=&Fort_Traps_Att
 bOverrideConcurrency=true
 maxDistance=6000
 pitchMultiplier=1.1
 soundClassObject=&Fort_Traps
 volumeMultiplier=0.7
 concurrencyOverrides {maxCount=1, resolutionRule=PreventNew}
}
