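# Sound cue for placing the Ice Trap, built from borrowed Freeze Trap and Ceiling Gas Trap waves.
# A quality switch sends its first two levels to a mixer layering a quieter random gas-activation hiss under the freeze-trap placement sound.
# The other two levels each play a single optimized freeze-placement wave with slight pitch and volume variation.
# Playback uses the shared traps attenuation and the trap placement sound class.
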SoundCue /Game/Sounds/Fort_Traps/Ice_Trap/IceTrap_Place_Cue
root=QualityLevel_0
nodes [
 {id=Mixer_0, children [Random_0,Modulator_0], properties {inputVolume [0.6,1]}}
 {id=Modulator_0, children [WavePlayer_0], properties {pitchMax=1, pitchMin=1, volumeMax=1.2, volumeMin=1}}
 {id=Modulator_1, children [WavePlayer_5], properties {pitchMax=1.1, pitchMin=0.9, volumeMax=1.1}}
 {id=Modulator_2, children [WavePlayer_6], properties {pitchMax=1.1, pitchMin=0.9, volumeMax=1.1}}
 {id=QualityLevel_0, children [Mixer_0,Mixer_0,Modulator_1,Modulator_2]}
 {id=Random_0, children [WavePlayer_2,WavePlayer_3], properties {weights [1,1]}}
 {id=WavePlayer_0, properties {soundWaveAssetPtr=/Game/Sounds/Fort_Traps/Freeze_Trap/Freeze_Trap_Place_01}}
 {id=WavePlayer_2, properties {soundWaveAssetPtr=/Game/Sounds/Fort_Traps/Ceiling_Gas_Trap/CeilingTrap_Gas_Activate_002}}
 {id=WavePlayer_3, properties {soundWaveAssetPtr=/Game/Sounds/Fort_Traps/Ceiling_Gas_Trap/CeilingTrap_Gas_Activate_001}}
 {id=WavePlayer_5, properties {soundWaveAssetPtr=/Game/Sounds/Fort_Traps/Freeze_Trap/Freeze_Trap_Place_opt}}
 {id=WavePlayer_6, properties {soundWaveAssetPtr=&Freeze_Trap_Place_opt}}
]
properties {
 attenuationSettings=&Fort_Traps_Att
 maxDistance=6000
 soundClassObject=/Game/Sounds/FortSoundClassesAndModes/Fort_Traps_Placement
 volumeMultiplier=0.8
}
