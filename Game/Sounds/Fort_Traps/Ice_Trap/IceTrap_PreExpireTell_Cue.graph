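# Sound cue for the Ice Trap's pre-expiry warning, a cracking tell that by its name plays before the trap expires.
# A Random node picks one of three warning-crack waves with equal weight and feeds a Modulator at the root.
# It plays in the in-game mission sound class with crystal shard attenuation and overrides concurrency so new instances are blocked once the limit is reached.
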SoundCue /Game/Sounds/Fort_Traps/Ice_Trap/IceTrap_PreExpireTell_Cue
root=Modulator_2
nodes [
 {id=Modulator_2, children [Random_0]}
 {id=Random_0, children [WavePlayer_0,WavePlayer_1,WavePlayer_2], properties {weights [1,1,1]}}
 {id=WavePlayer_0, properties {soundWaveAssetPtr=&/IceTrap_WarningCrack_01}}
 {id=WavePlayer_1, properties {soundWaveAssetPtr=&/IceTrap_WarningCrack_02}}
 {id=WavePlayer_2, properties {soundWaveAssetPtr=&/IceTrap_WarningCrack_03}}
]
properties {
 attenuationSettings=/Game/Sounds/Attenuation/Fort_Mission_CrystalShard_Att
 bOverrideConcurrency=true
 maxDistance=2800
 soundClassObject=/Game/Sounds/FortSoundClassesAndModes/Fort_Mission_IngameSounds
 volumeMultiplier=0.8
 attenuationOverrides {falloffDistance=2600}
 concurrencyOverrides {resolutionRule=PreventNew}
}
